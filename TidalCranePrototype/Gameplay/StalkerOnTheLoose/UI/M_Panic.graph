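# M_Panic is a translucent UI material, apparently a panic screen overlay for the StalkerOnTheLoose gameplay.
# An inverted exponential radial gradient clears the center and is multiplied by a steep power curve on offset V, weighting the effect toward the bottom edge.
# That mask multiplies two noise textures, bubble and cell, panned vertically in opposite directions at different speeds and tiling to create animated breakup.
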Material /TidalCranePrototype/Gameplay/StalkerOnTheLoose/UI/M_Panic
graph {
 properties {bAutomaticallySetUsageInEditor=false, blendMode=BLEND_Translucent, materialDomain=MD_UI, usedShadingModels=Default Lit}
 nodes [
  n1 = MaterialFunctionCall(Radius=Constant() {r=0.6}, Density=Constant() {r=8}) {materialFunction=&RadialGradientExponential}
  n2 = Add(A=ComponentMask(None=TextureCoordinate()) {g=true}, B=Constant() {r=0.1})
  n3 = Panner(Coordinate=TextureCoordinate()) {speedY=0.4}
  n4 = Panner(Coordinate=Multiply(A=TextureCoordinate()) {constB=0.5}) {speedY=-0.16}
  Multiply(
   A=Multiply(A=OneMinus(None=@n1.RadialGradientExponential), B=Power(Base=@n2) {constExponent=10})
   B=Multiply(
    A=TextureSample(UVs=@n3) {samplerType=SAMPLERTYPE_LinearGrayscale, texture=/Game/Global/Textures/Default/Noise/T_Bubble_B}.RGB
    B=TextureSample(UVs=@n4) {texture=/Game/Global/Textures/Default/Noise/T_CellNoise_01}.RGB
   )
  )
 ]
}
