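# Full-screen UMG overlay for the GnatGala stalker targeting ability, showing a dark purple panic effect.
# Two stretched images share the M_Panic material, with the top one rotated 180 degrees.
# Construct plays FadeIn4_0, which fades in only the bottom image's opacity; no animation touches the top image.
# Staged FadeIn1 to FadeIn4 and FadeOut animations also target the bottom image, but this graph never plays them.
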WidgetBlueprint /FNM_AI_GnatGala/BTInjectionBehaviour/Stalker_Targeting_Ability/StalkerTargetingAssets/UI/W_GnatGala_StalkTarget_UI
animations [
 FadeIn1 {
  binding_count=1
  length_seconds=1.25
  object=&self:FadeIn1
  sequence {bindings [ScreenEffectBottom {
     id=#1
     possessed_class=Image
     tracks [{
       type=MovieSceneColorTrack
       display_name=Color and Opacity
       sections [{
         type=MovieSceneColorSection
         end_frame=25
         start_frame=0
         channels [
          Color.R {type=MovieSceneScriptingFloatChannel, const [1,1]}
          Color.G {type=MovieSceneScriptingFloatChannel, const [1,1]}
          Color.B {type=MovieSceneScriptingFloatChannel, const [1,1]}
          Color.A {type=MovieSceneScriptingFloatChannel, keys [[0,0],[25,0.2]]}
         ]
        }]
      }]
    }]}
 }
 FadeIn2 {
  binding_count=1
  length_seconds=1.25
  object=&self:FadeIn2
  sequence {bindings [ScreenEffectBottom {
     id=#1
     possessed_class=Image
     tracks [{
       type=MovieSceneColorTrack
       display_name=Color and Opacity
       sections [{
         type=MovieSceneColorSection
         end_frame=25
         start_frame=0
         channels [
          Color.R {type=MovieSceneScriptingFloatChannel, const [1,1]}
          Color.G {type=MovieSceneScriptingFloatChannel, const [1,1]}
          Color.B {type=MovieSceneScriptingFloatChannel, const [1,1]}
          Color.A {type=MovieSceneScriptingFloatChannel, keys [[0,0.2],[25,0.4]]}
         ]
        }]
      }]
    }]}
 }
 FadeIn3 {
  binding_count=1
  length_seconds=1.25
  object=&self:FadeIn3
  sequence {bindings [ScreenEffectBottom {
     id=#1
     possessed_class=Image
     tracks [{
       type=MovieSceneColorTrack
       display_name=Color and Opacity
       sections [{
         type=MovieSceneColorSection
         end_frame=25
         start_frame=0
         channels [
          Color.R {type=MovieSceneScriptingFloatChannel, const [1,1]}
          Color.G {type=MovieSceneScriptingFloatChannel, const [1,1]}
          Color.B {type=MovieSceneScriptingFloatChannel, const [1,1]}
          Color.A {type=MovieSceneScriptingFloatChannel, keys [[0,0.4],[25,0.8]]}
         ]
        }]
      }]
    }]}
 }
 FadeIn4 {
  binding_count=1
  length_seconds=1.25
  object=&self:FadeIn4
  sequence {bindings [ScreenEffectBottom {
     id=#1
     possessed_class=Image
     tracks [{
       type=MovieSceneColorTrack
       display_name=Color and Opacity
       sections [{
         type=MovieSceneColorSection
         end_frame=25
         start_frame=0
         channels [
          Color.R {type=MovieSceneScriptingFloatChannel, const [1,1]}
          Color.G {type=MovieSceneScriptingFloatChannel, const [1,1]}
          Color.B {type=MovieSceneScriptingFloatChannel, const [1,1]}
          Color.A {type=MovieSceneScriptingFloatChannel, const [1,2]}
         ]
        }]
      }]
    }]}
 }
 FadeOut {
  binding_count=1
  length_seconds=1.25
  object=&self:FadeOut
  sequence {bindings [ScreenEffectBottom {
     id=#1
     possessed_class=Image
     tracks [{
       type=MovieSceneColorTrack
       display_name=Color and Opacity
       sections [{
         type=MovieSceneColorSection
         end_frame=25
         start_frame=0
         channels [
          Color.R {type=MovieSceneScriptingFloatChannel, const [1,1]}
          Color.G {type=MovieSceneScriptingFloatChannel, const [1,1]}
          Color.B {type=MovieSceneScriptingFloatChannel, const [1,1]}
          Color.A {type=MovieSceneScriptingFloatChannel, keys [[0,0.3],[25,0]]}
         ]
        }]
      }]
    }]}
 }
 FadeIn4_0 {
  binding_count=1
  length_seconds=2.5
  object=&self:FadeIn4_0
  sequence {bindings [ScreenEffectBottom {
     id=#1
     possessed_class=Image
     tracks [{
       type=MovieSceneColorTrack
       display_name=Color and Opacity
       sections [{
         type=MovieSceneColorSection
         end_frame=50
         start_frame=0
         channels [
          Color.R {type=MovieSceneScriptingFloatChannel, const [1,1]}
          Color.G {type=MovieSceneScriptingFloatChannel, const [1,1]}
          Color.B {type=MovieSceneScriptingFloatChannel, const [1,1]}
          Color.A {type=MovieSceneScriptingFloatChannel, keys [[0,0],[50,1]]}
         ]
        }]
      }]
    }]}
 }
]
graphs [EventGraph {nodes [
   Event Construct()
   > PlayAnimation(InAnimation=Get FadeIn4_0() -> (FadeIn4_0:Widget Animation Object Reference))
   Event PreConstruct()
   Event Tick()
  ]}]
variables []
widgets {
 info {inheritedWidgetCount=0, parentClass=&UserWidget, rootWidgetClass=&CanvasPanel, widgetCount=3}
 widgets [
  CanvasPanel_24 {widgetClassPath=&CanvasPanel, widget=CanvasPanel_24}
  ScreenEffectBottom {
   widgetClassPath=&Image
   bIsVariable=true
   parent=CanvasPanel_24
   widget=ScreenEffectBottom
   properties {brush {
     bIsDynamicallyLoaded=false
     drawAs=Image
     imageType=FullColor
     mirroring=NoMirror
     resourceName=None
     resourceObject=/TidalCranePrototype/Gameplay/StalkerOnTheLoose/UI/M_Panic
     tiling=NoTile
     imageSize=<1920 1080>
     margin=<ltrb 0 0 0 0>
     outlineSettings {
      bUseBrushTransparency=false
      roundingType=HalfHeightRadius
      width=0
      color {colorUseRule=UseColor_Specified, specifiedColor=<rgba 0 0 0 0>}
      cornerRadii=<0 0 0 0>
     }
     tintColor {colorUseRule=UseColor_Specified, specifiedColor=<rgba 0.0414 0.0046 0.0729 1>}
     uVRegion {bIsValid=false, max=<0 0>, min=<0 0>}
    }}
   slot {class=CanvasPanelSlot, properties {layoutData {alignment=<0 0>, anchors {maximum=<1 1>, minimum=<0 0>}, offsets=<ltrb 0 0 3.9239 1.0811>}}}
  }
  ScreenEffectTop {
   widgetClassPath=&Image
   bIsVariable=true
   parent=CanvasPanel_24
   widget=ScreenEffectTop
   properties {
    brush {
     bIsDynamicallyLoaded=false
     drawAs=Image
     imageType=FullColor
     mirroring=NoMirror
     resourceName=None
     resourceObject=&M_Panic
     tiling=NoTile
     imageSize=<1920 1080>
     margin=<ltrb 0 0 0 0>
     outlineSettings {
      bUseBrushTransparency=false
      roundingType=HalfHeightRadius
      width=0
      color {colorUseRule=UseColor_Specified, specifiedColor=<rgba 0 0 0 0>}
      cornerRadii=<0 0 0 0>
     }
     tintColor {colorUseRule=UseColor_Specified, specifiedColor=<rgba 0.0414 0.0046 0.0729 1>}
     uVRegion {bIsValid=false, max=<0 0>, min=<0 0>}
    }
    renderTransform {angle=180, scale=<1 1>, shear=<0 0>, translation=<0 0>}
   }
   slot {class=CanvasPanelSlot, properties {layoutData {alignment=<0 0>, anchors {maximum=<1 1>, minimum=<0 0>}, offsets=<ltrb 0 0 3.9239 1.0811>}}}
  }
 ]
}
